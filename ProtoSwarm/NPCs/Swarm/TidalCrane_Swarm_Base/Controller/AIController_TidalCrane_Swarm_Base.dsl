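; AI controller for the TidalCrane swarm base NPC that tracks whether its goal actor is within far or melee range.
; A goal-actor distance notify event switches on the notify index to choose the far-range or melee-range flag.
; When the incoming validity differs from the stored flag, it updates the flag and writes it to the matching Blackboard bool key, avoiding redundant writes.
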(event EventOnGoalActorDistanceNotify
  (params (NotifyIndex :type Integer) (IsValid :type Boolean))
  (switch int NotifyIndex
    (:0
      (if (not (== IsValid (Variables|Default|GetGoalActorDistance_IsInFarRange)))
        (bind _output_get (Variables|Default|SetGoalActorDistance_IsInFarRange IsValid))
        (AI|Components|Blackboard|SetValueasBool (AI|GetBlackboard self) (Variables|Default|GetGoalActorDistance_IsInFarRangeKey) _output_get)))
    (:1
      (if (not (== IsValid (Variables|Default|GetGoalActorDistance_IsInMeleeRange)))
        (bind _output_get_2 (Variables|Default|SetGoalActorDistance_IsInMeleeRange IsValid))
        (AI|Components|Blackboard|SetValueasBool (AI|GetBlackboard self) (Variables|Default|GetGoalActorDistance_IsInMeleeRangeKey) _output_get_2)))))

(fn ConstructionScript)
